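# Niagara glow effect for the golden CyberRunner backpack skin, with three persistent blue-to-purple particles spaced in a line on the skeletal mesh.
# A single burst spawns them, and each partition index drives offset, tint and a curve that gates the lights.
# A system `GL AMt` value, likely fed by User.GlitchAmount, blends sprite and light color, size and intensity toward a waveform-flickering state.
# Time-of-day scalars adjust brightness throughout.
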NiagaraSystem /BRCosmetics/Accessories/FORT_Backpacks/F_MED_CyberRunner_Backpack/Skins/Golden/FX/NS_CyberRunnerGolden
dependencies {
 usedDataInterfaces [&NiagaraDataInterfaceCurve, &NiagaraDataInterfaceSkeletalMesh]
 usedDynamicInputs []
 usedModules [
  &self:SystemSpawnScript.SystemScriptSource.SystemScriptGraph.NiagaraNodeAssignment_1.SetVariables_#1
  &SystemState
  &self:SystemSpawnScript.SystemScriptSource.SystemScriptGraph.NiagaraNodeAssignment_2.SetVariables_#2
 ]
}
emitters {Empty {
  emitterUpdateScript [
   &EmitterState()
   &SpawnBurst_Instantaneous(Spawn Count=3, Use Spawn Probability=false, UseLoopCountLimit=true)
  ]
  particleSpawnScript [
   &InitializeParticle(
    Color Mode=::NewEnumerator0
    Lifetime=5
    Sprite Size Mode=::NewEnumerator3
    Uniform Sprite Size=1
    UsePositionOffset=false
    Write Material Random=true
   )
   &PartitionParticles(
    Normalized Index In Partition Includes One=true
    Number Of Partitions=1
    Particles Per Partition=3
    Partition Mode:ENiagara_PartitionMode=::NewEnumerator1
   )
  ]
  particleUpdateScript [
   &ParticleState(Kill Particles When Lifetime Has Elapsed=false, Let Infinitely Lived Particles Die When Emitter Deactivates=true)
   &SkeletalMeshLocation(
    MaskSpawnGroup=false
    Mesh Sampling Type=::NewEnumerator5
    Sampled Position Offset=&Lerp_Vector(A=<-3.25 1.5 1>, B=<3.25 1.5 1>, Alpha=$Particles.PartitionParticles.Normalized Index In Partition)
    Skeletal Mesh=$System.Skeletal Mesh
    Socket Sampling Mode=::NewEnumerator5
   )
   &FN_Sprite_Attributes(
    Color=&LerpLinearColors(
     StartColor=&MultiplyLinearColor(
      Linear Color=&LerpLinearColors(StartColor=<rgba 0 0.3 1 0.48>, EndColor=<rgba 0.2 0 1 0.48>, LerpFactor=$Particles.PartitionParticles.Normalized Index In Partition)
      Scale Factor=&MakeFloat4FromVectorAndFloat(
       "Vector (RGB)"=&VectorFromFloat(Value=&FNTimeOfDayScalar(Night Scale=0.8, Morning Scale=2, Day Scale=2, Evening Scale=2))
       "Float (Alpha)"=0.8
      )
     )
     EndColor=&MultiplyLinearColor(
      Linear Color=&LerpLinearColors(StartColor=<rgba 0.01 0.307 1 1.5>, EndColor=<rgba 0.208 0.01 1 1.5>, LerpFactor=$Particles.PartitionParticles.Normalized Index In Partition)
      Scale Factor=&MakeFloat4FromVectorAndFloat(
       "Vector (RGB)"=&VectorFromFloat(Value=&Multiply_Float(
         A=&FNTimeOfDayScalar(Night Scale=0.5, Morning Scale=1, Evening Scale=1)
         B=&Waveform("Waveform [1]"=::NewEnumerator12, "[1] Frequency"=15, "[1] Amplitude Min/Max"=<0.4 1>)
        ))
       "Float (Alpha)"=1
      )
     )
     LerpFactor=$System.GL AMt
    )
    Create Color Attribute=true
    Create SpriteSize Attribute=true
    SpriteSize=&Vector2DFromFloat(Value=&Lerp_Float(
      A=20
      B=&FloatFromCurve(
       FloatCurve=DataInterface() {propertyValues {Curve {keys [<at 0 1>,<at 0.5 0.5>,<at 1 1>]}}}
       CurveIndex=$Particles.PartitionParticles.Normalized Index In Partition
       Scale Curve=30
      )
      Alpha=$System.GL AMt
     ))
   )
   &FN_Light_Attributes(
    Color=&LerpLinearColors(
     StartColor=&MultiplyLinearColor(
      Linear Color=&LerpLinearColors(StartColor=<rgba 0 0.3 1 0.5>, EndColor=<rgba 0.2 0 1 0.5>, LerpFactor=$Particles.PartitionParticles.Normalized Index In Partition)
      Scale Factor=&MakeFloat4FromVectorAndFloat(
       "Vector (RGB)"=&VectorFromFloat(Value=&FNTimeOfDayScalar(Night Scale=0.8, Morning Scale=0.6, Day Scale=0.6, Evening Scale=0.6))
       "Float (Alpha)"=0.8
      )
     )
     EndColor=&MultiplyLinearColor(
      Linear Color=&LerpLinearColors(StartColor=<rgba 0.2 0.44 1 1>, EndColor=<rgba 0.36 0.2 1 1>, LerpFactor=$Particles.PartitionParticles.Normalized Index In Partition)
      Scale Factor=&MakeFloat4FromVectorAndFloat(
       "Vector (RGB)"=&VectorFromFloat(Value=&Multiply_Float(
         A=&FNTimeOfDayScalar(Night Scale=0.7, Morning Scale=1, Evening Scale=1)
         B=&Waveform("Waveform [1]"=::NewEnumerator12, "[1] Frequency"=15, "[1] Amplitude Min/Max"=<0.4 1>)
        ))
       "Float (Alpha)"=1
      )
     )
     LerpFactor=$System.GL AMt
    )
    Create Color Attribute=true
    Create Exponent Attribute=true
    Create Radius Attribute=true
    Enable Particle Light Enable=true
    Exponent:NiagaraFloat=&Lerp_Float(A=1.5, B=2, Alpha=$System.GL AMt)
    Light Enabled:NiagaraBool=&SetBoolByFloatComparison(
     A=&FloatFromCurve(
      FloatCurve=DataInterface() {propertyValues {Curve {keys [<at 0 0>,<at 0.5 1>,<at 1 0>]}}}
      CurveIndex=$Particles.PartitionParticles.Normalized Index In Partition
     )
     B=0.5
    )
    Radius=&Lerp_Float(A=18, B=20, Alpha=$System.GL AMt)
    Scale Color=&Lerp_Float(
     A=&FNTimeOfDayScalar(Night Scale=8, Morning Scale=15, Day Scale=15, Evening Scale=15)
     B=&FNTimeOfDayScalar(Night Scale=15, Morning Scale=25, Day Scale=25, Evening Scale=25)
     Alpha=$System.GL AMt
    )
   )
   /SetVariables(Particles.Scale=&VectorFromFloat(Value=$Particles.FN_Light_Attributes.Radius)) {enabled=false, isSetParameters=true}
   /SetVariables(Particles.Linear Color=<rgba 1 1 1 0.5>) {enabled=false, isSetParameters=true}
  ]
  rendererClasses [
   &NiagaraSpriteRendererProperties
   &NiagaraLightRendererProperties
   &NiagaraMeshRendererProperties
  ]
  renderers [
   {rendererClass=&NiagaraSpriteRendererProperties}
   {rendererClass=&NiagaraLightRendererProperties, rendererIndex=1}
   {rendererClass=&NiagaraMeshRendererProperties, rendererIndex=2}
  ]
 }}
summary {
 systemName=NS_CyberRunnerGolden
 emitters [{
   emitterName=Empty
   rendererClasses [
    &NiagaraSpriteRendererProperties
    &NiagaraLightRendererProperties
    &NiagaraMeshRendererProperties
   ]
  }]
 userVariables [User.GlitchAmount {type=NiagaraFloat, defaultValue=0, description=""}]
}
