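# Additive unlit hologram material for the Golden skin of the CyberRunner backpack, with a Blueprint-driven glitch for elimination reactions.
# Packed texture channels drive base and edge color blends, and MasterElimReactionLerp fades to a glitched palette with chromatic aberration from offset resamples.
# Rotated scanlines blend low and high opacities.
# World position offset shakes the mesh using summed sines and panned gradient-texture distortion, before a TechArt accessory override function.
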Material /BRCosmetics/Accessories/FORT_Backpacks/F_MED_CyberRunner_Backpack/Skins/Golden/Materials/M_CyberRunner_Golder_Hologram
graph {
 properties {
  bUseMaterialAttributes=true
  bUsedWithMeshDeformer=true
  bUsedWithSkeletalMesh=true
  blendMode=BLEND_Additive
  ditherOpacityMask=true
  pixelDepthOffsetMode=PDOM_Legacy
  shadingModel=MSM_Unlit
  usedShadingModels=Unlit
 }
 nodes [
  n1 = TextureSample() {texture=&/golden_hologram}
  n2 = VectorParameter() {group=Base Colors, parameterName=Low Color Options, channelNames {a=Contrast, b=add, g=Mul, r=power}, defaultValue=<rgba 1 1 0 0>}
  n3 = Add(A=Multiply(A=Power(Base=@n1.A, Exp=@n2.power), B=@n2.Mul), B=@n2.add)
  n4 = VectorParameter() {group=Base Colors, parameterName=High Color Options, channelNames {a=Contrast, b=add, g=Mul, r=power}, defaultValue=<rgba 1 1 0 0>}
  n5 = Add(A=Multiply(A=Power(Base=@n1.A, Exp=@n4.power), B=@n4.Mul), B=@n4.add)
  n6 = LinearInterpolate(
   A=LinearInterpolate(
    A=VectorParameter() {group=Base Colors, parameterName=Low Color One, defaultValue=<rgba 1 0 0 0>}.RGB
    B=VectorParameter() {group=Base Colors, parameterName=Low Color two, defaultValue=<rgba 0 1 0 0>}.RGB
    Alpha=MaterialFunctionCall(In=@n3, Contrast=@n2.Contrast) {materialFunction=&CheapContrast}.Result
   )
   B=LinearInterpolate(
    A=VectorParameter() {group=Base Colors, parameterName=High Color One, defaultValue=<rgba 0 0 1 0>}.RGB
    B=VectorParameter() {group=Base Colors, parameterName=High Color two, defaultValue=<rgba 1 1 0 0>}.RGB
    Alpha=MaterialFunctionCall(In=@n5, Contrast=@n4.Contrast) {materialFunction=&CheapContrast}.Result
   )
   Alpha=@n1.B
  ) {constB=-2.0016}
  n7 = VectorParameter() {group=Base Colors, parameterName=Edge Color, channelNames {a=Use just Edge Color, b="", g="", r=""}, defaultValue=<rgba 1 1 1 0>}
  n8 = LinearInterpolate(A=@n6, B=LinearInterpolate(A=Multiply(A=@n7, B=@n6), B=@n7, Alpha=@n7), Alpha=@n1.R)
  n9 = TextureSample() {texture=&golden_hologram}
  n10 = VectorParameter() {group=glitch, parameterName=Glitched Low Color Options, channelNames {a=Contrast, b=add, g=Mul, r=power}, defaultValue=<rgba 1 1 0 0>}
  n11 = Add(A=Multiply(A=Power(Base=@n9.A, Exp=@n10.power), B=@n10.Mul), B=@n10.add)
  n12 = Multiply(A=VectorParameter() {group=Chromaic abbers, parameterName=chrome color red, defaultValue=<rgba 1 0 0 0>}.RGB, B=@n9.B)
  n13 = ScalarParameter() {defaultValue=0.08, group=Chromaic abbers, parameterName=Chromatic Offset}
  n14 = TextureSample(UVs=Panner(Time=@n13) {speedY=0.1}) {texture=&golden_hologram}
  n15 = Multiply(A=VectorParameter() {group=Chromaic abbers, parameterName=chrome color Green, defaultValue=<rgba 0 1 0 0>}.RGB, B=@n14.B)
  n16 = VectorParameter() {group=Chromaic abbers, parameterName=chrome color Blue, defaultValue=<rgba 0 0 1 0>}
  n17 = TextureSample(UVs=Panner(Time=@n13) {speedY=-0.1}) {texture=&golden_hologram}
  n18 = Multiply(
   A=MaterialFunctionCall(X=@n9.B, Y=@n14.B, Z=@n17.B) {materialFunction=&MakeFloat3}.Result
   B=VectorParameter() {group=Chromaic abbers, parameterName=Chromatic Masks, defaultValue=<rgba 1 1 1 0>}.RGB
  )
  n19 = Constant() {r=1}
  n20 = LinearInterpolate(
   A=LinearInterpolate(
    A=VectorParameter() {group=glitch, parameterName=Glitched Low Color One, defaultValue=<rgba 1 0 0 0>}.RGB
    B=VectorParameter() {group=glitch, parameterName=Glitched Low Color two, defaultValue=<rgba 0 1 0 0>}.RGB
    Alpha=MaterialFunctionCall(In=@n11, Contrast=@n10.Contrast) {materialFunction=&CheapContrast}.Result
   )
   B=Add(A=Add(A=@n12, B=@n15), B=Multiply(A=@n16.RGB, B=@n17.B))
   Alpha=Saturate(None=DotProduct(A=@n18, B=@n19))
  ) {constB=-2.0016}
  n21 = VectorParameter() {group=glitch, parameterName=Glitched Edge Color, channelNames {a=Use just Edge Color, b="", g="", r=""}, defaultValue=<rgba 1 1 1 0>}
  n22 = LinearInterpolate(A=@n20, B=LinearInterpolate(A=Multiply(A=@n21, B=@n20), B=@n21, Alpha=@n21), Alpha=@n9.R)
  n23 = NamedRerouteDeclaration(None=ScalarParameter() {group=Blueprint Driven, parameterName=MasterElimReactionLerp}) {name=Param MasterElimReactionLerp, nodeColor=<rgba 0 1 0.0391 1>}
  n24 = NamedRerouteUsage()
  n25 = LinearInterpolate(
   A=ScalarParameter() {defaultValue=1, parameterName=Opacity Low}
   B=ScalarParameter() {defaultValue=1, group=glitch, parameterName=Glitch Opacity Low}
   Alpha=@n24
  )
  n26 = LinearInterpolate(
   A=ScalarParameter() {defaultValue=1, parameterName=Opacity High}
   B=ScalarParameter() {defaultValue=1, group=glitch, parameterName=Glitch Opacity High}
   Alpha=@n24
  )
  n27 = VectorParameter() {parameterName=Scanline Options, channelNames {a=Speed 2, b=Tile 2, g=Speed 1, r=Tile 1}, defaultValue=<rgba 2 -0.5 40 0.5>}
  n28 = TextureCoordinate()
  n29 = ScalarParameter() {defaultValue=0.25, parameterName=RotScanlines}
  n30 = MaterialFunctionCall!(
   ScanLine_Tiling_1=@n27.Tile 1
   Scanline_PannerSpeed_1=@n27.Tile 1
   ScanLine_Tiling=@n27.Tile 1
   Scanline_PannerSpeed=@n27.Tile 1
   UVs=NamedRerouteDeclaration(None=MaterialFunctionCall(UVs=@n28, "Rotation Angle (0-1)"=@n29) {materialFunction=&CustomRotator}.Rotated Values) {name=RotatedUVs, nodeColor=<rgba 1 0.4922 0 1>}
  ) {materialFunction=/Game/Weapons/WeaponSkins/MaterialFunction/ScanLines}
  n31 = Multiply(A=Time(), B=ScalarParameter() {defaultValue=3, group=WPO, parameterName=Overall Shake Master Speed})
  n32 = Multiply(A=@n31) {constB=0.91}
  n33 = Multiply(A=@n31) {constB=1.4741}
  n34 = Multiply(A=@n31) {constB=0.4996}
  n35 = Add(A=Add(A=Sine(None=@n31), B=Sine(None=@n32)), B=Add(A=Sine(None=@n33), B=Sine(None=@n34)))
  n36 = Add(A=Sine(None=Multiply(A=@n31) {constA=-4.906, constB=2.7578}), B=Sine(None=Multiply(A=@n31) {constB=0.78}))
  n37 = VectorParameter() {parameterName=Glitchy Distortion SPeeds Tile, defaultValue=<rgba 1 2 3 1>}
  n38 = Multiply(A=TextureCoordinate(), B=@n37.A)
  n39 = TextureSample(UVs=Panner(Coordinate=@n38, Speed=AppendVector(A=@n37.R, B=@n37.R)) {speedX=1}) {
   samplerType=SAMPLERTYPE_Masks
   texture=&/T_Gradiant_WPO
  }
  n40 = TextureSample(UVs=Panner(Coordinate=@n38, Speed=AppendVector(A=@n37.G, B=@n37.G)) {speedX=2}) {
   samplerType=SAMPLERTYPE_Masks
   texture=&T_Gradiant_WPO
  }
  n41 = Add(A=ComponentMask(None=Subtract(A=@n39.RGB) {constB=0.5}) {r=true}, B=ComponentMask(None=Subtract(A=@n40.RGB) {constB=0.5}) {g=true})
  n42 = TextureSample(UVs=Panner(Coordinate=@n38, Speed=AppendVector(A=@n37.B, B=@n37.B)) {speedX=3}) {
   samplerType=SAMPLERTYPE_Masks
   texture=&T_Gradiant_WPO
  }
  n43 = ComponentMask(None=Subtract(A=@n42.RGB) {constB=0.5}) {b=true}
  n44 = Add(
   A=Multiply(A=Add(A=@n35, B=@n36), B=ScalarParameter() {defaultValue=1, group=WPO, parameterName=Shake WPO Str})
   B=Multiply(A=Add(A=@n41, B=@n43), B=ScalarParameter() {defaultValue=1, parameterName=Glitchy Distortion Str})
  ) {constB=0}
  n45 = Multiply(
   A=Transform(None=MaterialFunctionCall(X=@n44) {materialFunction=&MakeFloat3}.Result) {transformSourceType=TRANSFORMSOURCE_Local}
   B=ScalarParameter() {defaultValue=1, group=WPO, parameterName=Overall WPO Str}
  )
  n46 = NamedRerouteUsage()
  MaterialFunctionCall(In=SetMaterialAttributes!(
    MaterialAttributes
    Emissive Color=LinearInterpolate(A=@n8, B=@n22, Alpha=@n23)
    Opacity=LinearInterpolate(A=@n25, B=@n26, Alpha=@n30.Result)
    World Position Offset=Multiply(A=@n45, B=@n46)
   ) {attributeSetTypes [#1, #2, #3]}) {materialFunction=&MF_TechArtOverrides_accessories}
  NamedRerouteUsage()
 ]
}
